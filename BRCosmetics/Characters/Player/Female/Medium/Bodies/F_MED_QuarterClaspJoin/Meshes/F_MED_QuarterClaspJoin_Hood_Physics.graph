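# Physics asset for the hood mesh of the F_MED_QuarterClaspJoin female player cosmetic body, defining collision and joint limits for hood secondary motion.
# Left and right hood bones each get one capsule body, while spine_05 carries two capsules that form the upper-torso volume.
# Each hood body is constrained directly to spine_05 with mirrored setup, asymmetric swing limits and a very tight twist limit.
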PhysicsAsset /BRCosmetics/Characters/Player/Female/Medium/Bodies/F_MED_QuarterClaspJoin/Meshes/F_MED_QuarterClaspJoin_Hood_Physics
bodies {
 dyn_hood_l [{shapeType=Capsule, length=6.4501, radius=5.6834, shapeName=dyn_hood_l_capsule, center=<5.4 0 4>, rotation=<rot 0 0 -90>}]
 dyn_hood_r [{shapeType=Capsule, length=6.4501, radius=5.6834, shapeName=dyn_hood_r_capsule, center=<5.4 0 4>, rotation=<rot 0 0 90>}]
 spine_05 [
  {shapeType=Capsule, length=21.25, radius=8.75, center=<0.8141 1.5922 -0.0002>}
  {shapeType=Capsule, length=21.1644, radius=8.8356, shapeName=spine_05_capsule, center=<-5.6153 0.2499 0.3165>}
 ]
}
constraints [
 {bone1Name=dyn_hood_l, bone2Name=spine_05, swing1LimitDegrees=35, swing2LimitDegrees=60, twistLimitDegrees=3}
 {bone1Name=dyn_hood_r, bone2Name=spine_05, swing1LimitDegrees=35, swing2LimitDegrees=60, twistLimitDegrees=3}
]
