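# Physics asset for the F_MED_QuarterClaspJoin female medium player body, apparently for secondary motion of its jacket.
# Capsule bodies cover the dyn_jacket bones, with two more capsules on spine_03 and one on a body with bone name None.
# The back, front, left and right jacket bodies are each constrained to spine_03 with narrow swing limits and zero twist limit.
# The dyn_jacket_01 bodies and the None body have no constraints.
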PhysicsAsset /BRCosmetics/Characters/Player/Female/Medium/Bodies/F_MED_QuarterClaspJoin/Meshes/F_MED_QuarterClaspJoin_Physics
bodies {
 None [{shapeType=Capsule, length=13.1673, radius=6.4907, center=<-0.1728 1.1768 2.7741>, rotation=<rot -87.7548 12.0543 -14.295>}]
 dyn_jacket_01_b [{
   shapeType=Capsule
   length=6.027
   radius=4.5967
   shapeName=dyn_jacket_01_b_capsule
   center=<0.3213 0.0239 2.1463>
   rotation=<rot 0 0.1556 -89.8085>
  }]
 dyn_jacket_01_f [{
   shapeType=Capsule
   length=4.5698
   radius=2.9746
   shapeName=dyn_jacket_01_f_capsule
   center=<4.4879 0.0903 1.9293>
   rotation=<rot 0 -179.6366 -89.6476>
  }]
 dyn_jacket_01_l [{shapeType=Capsule, length=6.4415, radius=2.8588, shapeName=dyn_jacket_01_l_capsule, center=<1.5702 0.0574 1.2553>, rotation=<rot 0 180 -90>}]
 dyn_jacket_back [{shapeType=Capsule, length=5.6902, radius=4.3667, shapeName=dyn_jacket_back_capsule, center=<3.5122 -1.5 0>}]
 dyn_jacket_front [{shapeType=Capsule, length=10.4115, radius=4.2446, shapeName=dyn_jacket_front_capsule, center=<4 4.2 0>}]
 dyn_jacket_l [{
   shapeType=Capsule
   length=10.5065
   radius=4.5855
   shapeName=dyn_jacket_l_capsule
   center=<3.0678 0.0147 2.192>
   rotation=<rot 0 0.0045 90.3789>
  }]
 dyn_jacket_r [{
   shapeType=Capsule
   length=10.5065
   radius=4.5855
   shapeName=dyn_jacket_r_capsule
   center=<3.0678 -0.0144 -2.1393>
   rotation=<rot 0 0.0045 -89.6211>
  }]
 spine_03 [
  {shapeType=Capsule, length=10.7753, radius=8.4747, center=<-0.1292 2.7879 0.0001>, rotation=<rot 90 5 0>}
  {shapeType=Capsule, length=6.9007, radius=6.0286, center=<-2.9695 0.597 -0.0254>, rotation=<rot 0.2247 7.1286 -0.1929>}
 ]
}
constraints [
 {bone1Name=dyn_jacket_back, bone2Name=spine_03, swing1LimitDegrees=10, swing2LimitDegrees=10, twistLimitDegrees=0}
 {bone1Name=dyn_jacket_front, bone2Name=spine_03, swing1LimitDegrees=10, swing2LimitDegrees=10, twistLimitDegrees=0}
 {bone1Name=dyn_jacket_l, bone2Name=spine_03, swing1LimitDegrees=10, swing2LimitDegrees=10, twistLimitDegrees=0}
 {bone1Name=dyn_jacket_r, bone2Name=spine_03, swing1LimitDegrees=10, swing2LimitDegrees=10, twistLimitDegrees=0}
]
