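# Physics asset for the female medium QuarterClaspJoin cosmetic body, rigging its left and right strings and a grenade for secondary motion.
# Two three-link capsule chains hang from dyn_string_parent, a compound body of four larger capsules.
# Chain joints allow modest swing and almost no twist, with the right side looser at the root and stiffening toward the tip.
# The grenade capsule attaches separately to the right thigh twist bone with asymmetric swing limits.
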PhysicsAsset /BRCosmetics/Characters/Player/Female/Medium/Bodies/F_MED_QuarterClaspJoin/Meshes/F_MED_QuarterClaspJoin_String_Physics
bodies {
 dyn_grenade [{shapeType=Capsule, length=7.3551, radius=3.9119, center=<3.5046 -0.3003 1.0558>, rotation=<rot -2.6524 -90 3.8034>}]
 dyn_string_l_01 [{shapeType=Capsule, length=2.755, radius=0.505, center=<1 0 0>, rotation=<rot 90 0 0>}]
 dyn_string_l_02 [{shapeType=Capsule, length=2.755, radius=0.505, center=<1 0 0>, rotation=<rot 90 0 0>}]
 dyn_string_l_03 [{shapeType=Capsule, length=2.755, radius=0.505, center=<1 0 0>, rotation=<rot 90 0 0>}]
 dyn_string_parent [
  {shapeType=Capsule, length=11.5, radius=5.75, center=<5.1179 8.4847 -4.6808>, rotation=<rot -84.9275 -180 180>}
  {shapeType=Capsule, length=11.5, radius=5.75, center=<5.2356 -7.0949 -3.5776>, rotation=<rot -84.9312 -180 180>}
  {shapeType=Capsule, length=8.25, radius=9, center=<10.4738 0.1383 -6.4267>, rotation=<rot -75.8899 -171.0966 81.0147>}
  {shapeType=Capsule, length=14.0804, radius=4.0044, shapeName=dyn_string_parent_capsule, center=<2.4066 0 -1.3689>, rotation=<rot 0 0 -90>}
 ]
 dyn_string_r_01 [{shapeType=Capsule, length=2.755, radius=0.505, center=<1 0 0>, rotation=<rot 90 0 -5>}]
 dyn_string_r_02 [{shapeType=Capsule, length=2.755, radius=0.505, center=<1 0 0>, rotation=<rot 90 0 -5>}]
 dyn_string_r_03 [{shapeType=Capsule, length=4.005, radius=0.505, center=<1 0 0>, rotation=<rot 90 0 -5>}]
 thigh_twist_01_r [{shapeType=Capsule, length=10.8739, radius=4.4558, shapeName=thigh_twist_01_r_capsule, rotation=<rot -89.155 -179.9994 179.9991>}]
}
constraints [
 {bone1Name=dyn_string_r_01, bone2Name=dyn_string_parent, swing1LimitDegrees=25, swing2LimitDegrees=25, twistLimitDegrees=1}
 {bone1Name=dyn_string_r_02, bone2Name=dyn_string_r_01, swing1LimitDegrees=20, swing2LimitDegrees=20, twistLimitDegrees=1}
 {bone1Name=dyn_string_r_03, bone2Name=dyn_string_r_02, swing1LimitDegrees=15, swing2LimitDegrees=15, twistLimitDegrees=1}
 {bone1Name=dyn_string_l_01, bone2Name=dyn_string_parent, swing1LimitDegrees=15, swing2LimitDegrees=15, twistLimitDegrees=1}
 {bone1Name=dyn_string_l_02, bone2Name=dyn_string_l_01, swing1LimitDegrees=15, swing2LimitDegrees=15, twistLimitDegrees=1}
 {bone1Name=dyn_string_l_03, bone2Name=dyn_string_l_02, swing1LimitDegrees=15, swing2LimitDegrees=15, twistLimitDegrees=1}
 {bone1Name=dyn_grenade, bone2Name=thigh_twist_01_r, swing1LimitDegrees=25, swing2LimitDegrees=10, twistLimitDegrees=3}
]
