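# Physics asset for the back dynamic hair on the QuarterClaspJoin medium female character's face accessory part.
# Capsules on upper spine, clavicles, neck and head form the body shapes around four hair chains (back left, middle, right, mid-left).
# Each chain hangs from the head by a swing-free base, with tight swing limits and zero twist further down.
# Angularly free cross-constraints tie mid-left to back-left and back-right to back-middle.
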PhysicsAsset /BRCosmetics/Characters/Player/Female/Medium/Bodies/F_MED_QuarterClaspJoin/Meshes/Parts/F_MED_QuarterClaspJoin_Back_FaceAcc_Physics
bodies {
 clavicle_l [{shapeType=Capsule, length=12, radius=5.5, shapeName=clavicle_l_capsule, center=<3.5 -1 -1>, rotation=<rot 80 -0.0003 -0.0003>}]
 clavicle_r [{shapeType=Capsule, length=12, radius=5.5, shapeName=clavicle_r_capsule, center=<-3.5 1 1>, rotation=<rot 80 -0.0003 -0.0004>}]
 dyn_hair_bk_02 [{
   shapeType=Capsule
   length=3.8869
   radius=2.8631
   shapeName=dyn_hair_bk_02_capsule
   center=<4.4511 1.2932 -0.2195>
   rotation=<rot 79.9972 -166.5116 180>
  }]
 dyn_hair_bk_l_01 [{shapeType=Capsule, length=2.6556, radius=3.1062, shapeName=dyn_hair_bk_l_01_capsule, center=<3.6 0 -0.27>, rotation=<rot -0.9 -90 -88>}]
 dyn_hair_bk_l_02 [{
   shapeType=Capsule
   length=3.4394
   radius=2.6199
   shapeName=dyn_hair_bk_l_02_capsule
   center=<3.4821 -0.2724 -0.5195>
   rotation=<rot -13.74 -96 -86>
  }]
 dyn_hair_bk_l_base [{shapeType=Capsule, length=0.5, radius=0.5, shapeName=dyn_hair_bk_l_base_capsule}]
 dyn_hair_bk_l_parent [{
   shapeType=Capsule
   length=2.059
   radius=2.7616
   shapeName=dyn_hair_bk_l_parent_capsule
   center=<2.72 0.42 0.81>
   rotation=<rot -2.4 -85 -110>
  }]
 dyn_hair_bk_mid_01 [{
   shapeType=Capsule
   length=3.0989
   radius=2.425
   shapeName=dyn_hair_bk_mid_01_capsule
   center=<2.8663 0.2812 -0.1756>
   rotation=<rot 67.3018 -67.4891 -69.7649>
  }]
 dyn_hair_bk_mid_base [{shapeType=Capsule, length=0.5, radius=0.5, shapeName=dyn_hair_bk_mid_base_capsule}]
 dyn_hair_bk_mid_parent [{
   shapeType=Capsule
   length=1.5248
   radius=2.2291
   shapeName=dyn_hair_bk_mid_parent_capsule
   center=<2.1 0.12 0.6>
   rotation=<rot 18.7416 -89.207 -100.3277>
  }]
 dyn_hair_bk_r_01 [{shapeType=Capsule, length=3.3295, radius=2, shapeName=dyn_hair_bk_r_01_capsule, center=<2.8 -1.64 -0.72>, rotation=<rot 63.2249 90 89.9999>}]
 dyn_hair_bk_r_02 [{shapeType=Capsule, length=3.4749, radius=2.0251, shapeName=dyn_hair_bk_r_02_capsule, center=<4.1867 -0.1104 0.1534>, rotation=<rot 90 0 0>}]
 dyn_hair_bk_r_base [{shapeType=Capsule, length=0.5, radius=0.5, shapeName=dyn_hair_bk_r_base_capsule}]
 dyn_hair_bk_r_parent [{
   shapeType=Capsule
   length=1.5827
   radius=2.2983
   shapeName=dyn_hair_bk_r_parent_capsule
   center=<2.152 -0.3209 -0.044>
   rotation=<rot -37.3028 -105.6519 -77.2143>
  }]
 dyn_hair_mid_l_01 [{shapeType=Capsule, length=3.7585, radius=1.3171, shapeName=dyn_hair_mid_l_01_capsule, center=<3 0 0>, rotation=<rot 90 0 0>}]
 dyn_hair_mid_l_base [{shapeType=Capsule, length=0.5, radius=0.5, shapeName=dyn_hair_mid_l_base_capsule}]
 dyn_hair_mid_l_parent [{
   shapeType=Capsule
   length=2.5936
   radius=2.227
   shapeName=dyn_hair_bk_l_parent_capsule
   center=<2.6917 0.4312 0.485>
   rotation=<rot -2.4 -84.9999 -85.2246>
  }]
 head [{shapeType=Capsule, length=6, radius=6, shapeName=head_capsule, center=<3.5 3 0>, rotation=<rot 90 0 0>}]
 neck_01 [
  {
   shapeType=Capsule
   length=23.3359
   radius=4.259
   shapeName=neck_01_capsule
   center=<3.6571 1.2731 -0.811>
   rotation=<rot 2.5902 179.9997 -180>
  }
  {shapeType=Capsule, length=9.0313, radius=5.6256, shapeName=neck_01_capsule, center=<4.2664 8.9943 0>}
 ]
 spine_05 [
  {shapeType=Capsule, length=12, radius=9.25, shapeName=spine_05_capsule, center=<3.7117 2.4053 0>}
  {shapeType=Capsule, length=12, radius=7.75, shapeName=spine_05_capsule, center=<-6.4845 2.8778 0>}
 ]
}
constraints [
 {bone1Name=dyn_hair_mid_l_parent, bone2Name=dyn_hair_mid_l_base, swing1LimitDegrees=5, swing2LimitDegrees=5, twistLimitDegrees=0}
 {bone1Name=dyn_hair_bk_l_parent, bone2Name=dyn_hair_bk_l_base, swing1LimitDegrees=5, swing2LimitDegrees=5, twistLimitDegrees=0}
 {bone1Name=dyn_hair_mid_l_01, bone2Name=dyn_hair_mid_l_parent, swing1LimitDegrees=5, swing2LimitDegrees=15, twistLimitDegrees=0}
 {bone1Name=dyn_hair_bk_l_01, bone2Name=dyn_hair_bk_l_parent, swing1LimitDegrees=5, swing2LimitDegrees=15, twistLimitDegrees=0}
 {bone1Name=dyn_hair_bk_l_02, bone2Name=dyn_hair_bk_l_01, swing1LimitDegrees=5, swing2LimitDegrees=15, twistLimitDegrees=0}
 {bone1Name=dyn_hair_bk_mid_parent, bone2Name=dyn_hair_bk_mid_base, swing1LimitDegrees=5, swing2LimitDegrees=5, twistLimitDegrees=0}
 {bone1Name=dyn_hair_bk_mid_01, bone2Name=dyn_hair_bk_mid_parent, swing1LimitDegrees=5, swing2LimitDegrees=15, twistLimitDegrees=0}
 {bone1Name=dyn_hair_bk_02, bone2Name=dyn_hair_bk_mid_01, swing1LimitDegrees=5, swing2LimitDegrees=15, twistLimitDegrees=0}
 {bone1Name=dyn_hair_bk_r_parent, bone2Name=dyn_hair_bk_r_base, swing1LimitDegrees=5, swing2LimitDegrees=5, twistLimitDegrees=0}
 {bone1Name=dyn_hair_bk_r_01, bone2Name=dyn_hair_bk_r_parent, swing1LimitDegrees=5, swing2LimitDegrees=15, twistLimitDegrees=0}
 {bone1Name=dyn_hair_bk_r_02, bone2Name=dyn_hair_bk_r_01, swing1LimitDegrees=5, swing2LimitDegrees=15, twistLimitDegrees=0}
 {
  bone1Name=clavicle_l
  bone2Name=spine_05
  swing1LimitDegrees=45
  swing1Motion=Free
  swing2LimitDegrees=45
  swing2Motion=Free
  twistLimitDegrees=45
  twistMotion=Free
 }
 {
  bone1Name=clavicle_r
  bone2Name=spine_05
  swing1LimitDegrees=45
  swing1Motion=Free
  swing2LimitDegrees=45
  swing2Motion=Free
  twistLimitDegrees=45
  twistMotion=Free
 }
 {
  bone1Name=neck_01
  bone2Name=spine_05
  swing1LimitDegrees=45
  swing1Motion=Free
  swing2LimitDegrees=45
  swing2Motion=Free
  twistLimitDegrees=45
  twistMotion=Free
 }
 {
  bone1Name=head
  bone2Name=neck_01
  swing1LimitDegrees=45
  swing1Motion=Free
  swing2LimitDegrees=45
  swing2Motion=Free
  twistLimitDegrees=45
  twistMotion=Free
 }
 {
  bone1Name=dyn_hair_bk_l_base
  bone2Name=head
  swing1LimitDegrees=45
  swing1Motion=Free
  swing2LimitDegrees=45
  swing2Motion=Free
  twistLimitDegrees=45
  twistMotion=Free
 }
 {
  bone1Name=dyn_hair_bk_mid_base
  bone2Name=head
  swing1LimitDegrees=45
  swing1Motion=Free
  swing2LimitDegrees=45
  swing2Motion=Free
  twistLimitDegrees=45
  twistMotion=Free
 }
 {
  bone1Name=dyn_hair_bk_r_base
  bone2Name=head
  swing1LimitDegrees=45
  swing1Motion=Free
  swing2LimitDegrees=45
  swing2Motion=Free
  twistLimitDegrees=45
  twistMotion=Free
 }
 {
  bone1Name=dyn_hair_mid_l_base
  bone2Name=head
  swing1LimitDegrees=45
  swing1Motion=Free
  swing2LimitDegrees=45
  swing2Motion=Free
  twistLimitDegrees=45
  twistMotion=Free
 }
 {
  bone1Name=dyn_hair_mid_l_01
  bone2Name=dyn_hair_bk_l_01
  swing1LimitDegrees=45
  swing1Motion=Free
  swing2LimitDegrees=45
  swing2Motion=Free
  twistLimitDegrees=45
  twistMotion=Free
 }
 {
  bone1Name=dyn_hair_bk_r_01
  bone2Name=dyn_hair_bk_mid_01
  swing1LimitDegrees=45
  swing1Motion=Free
  swing2LimitDegrees=45
  swing2Motion=Free
  twistLimitDegrees=45
  twistMotion=Free
 }
]
