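# Post-process Control Rig for the QuarterClaspJoin female face accessory that reshapes its dynamic hair bones to suit the character's pose.
# A sequence layers a constant adjustment with bone offsets weighted by head direction and aim, crouch, jump, tactical jump and skydive variables, some smoothed by AlphaInterp.
# A spherical pose reader on the left upper arm drives further corrections, combined with emote and head-up weights.
# A FrontEnd flag switches on separate offsets.
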ControlRigBlueprint /BRCosmetics/Characters/Player/Female/Medium/Bodies/F_MED_QuarterClaspJoin/Meshes/Parts/F_MED_QuarterClaspJoin_FaceAcc_Post_CtrlRig
graphs [
 RigVMFunctionLibrary {links [], nodes []}
 RigVMFunction_Sequence_1_ContainedGraph {nodes [
   n1 = RigVMFunctionEntryNode()
   RigVMFunction_Sequence = &RigVMFunction_Sequence(ExecuteContext=@n1.ExecuteContext)
   n2 = &RigVMFunction_Sequence(ExecuteContext=@RigVMFunction_Sequence.B)
   n3 = &RigVMFunction_Sequence(ExecuteContext=@n2.B)
   n4 = &RigVMFunction_Sequence(ExecuteContext=@n3.B)
   n5 = &RigVMFunction_Sequence(ExecuteContext=@n4.B)
   RigVMFunctionReturnNode(A=@RigVMFunction_Sequence.A, B=@n2.A, C=@n3.A, D=@n4.A, E=@n5.A, F=@n5.B)
  ]}
 RigVMModel {nodes [
   RigUnit_BeginExecution = &RigUnit_BeginExecution()
   RigVMFunction_Sequence = RigVMAggregateNode(ExecuteContext=@RigUnit_BeginExecution.ExecutePin)
   n1 = RigVMVariableNode() {variable=Head_L}
   &RigUnit_ModifyTransforms(
    ItemToModify="((Item=(Type=Bone,Name=\"dyn_hair_bk_l_parent\"),Transform=(Rotation=(X=0,Y=0,Z=0.0872,W=0.9962),Translation=(X=0,Y=3,Z=1),Scale3D=(X=1,Y=1,Z=1))),(Item=(Type=Bone,Name=\"dyn_hair_bk_l_01\"),Transform=(Rotation=(X=0,Y=0,Z=-0.1305,W=0.9914),Translation=(X=0,Y=0 ...(truncated)"
    WeightMinimum=0
    Weight=@n1.Value
    ExecutePin=@RigVMFunction_Sequence.B
   )
   > &RigUnit_ModifyTransforms(
    ItemToModify="((Item=(Type=Bone,Name=\"dyn_hair_bk_mid_parent\"),Transform=(Rotation=(X=0,Y=0,Z=0,W=1),Translation=(X=0,Y=4,Z=2),Scale3D=(X=0.8,Y=1,Z=1))),(Item=(Type=Bone,Name=\"dyn_hair_bk_mid_01\"),Transform=(Rotation=(X=0.0227,Y=0.1285,Z=-0.1722,W=0.9764),Translation=(X ...(truncated)"
    WeightMinimum=0
    Weight=@n1.Value
   )
   > &RigUnit_ModifyTransforms(
    ItemToModify=((Item=(Type=Bone,Name="dyn_hair_01"),Transform=(Rotation=(X=0,Y=0,Z=0,W=1),Translation=(X=0,Y=0,Z=0),Scale3D=(X=1,Y=1,Z=1))),(Item=(Type=Bone,Name="dyn_hair_02"),Transform=(Rotation=(X=0,Y=0,Z=0,W=1),Translation=(X=0,Y=0,Z=0),Scale3D=(X=1,Y=1,Z=1))))
    Weight=1
    WeightMinimum=0
   )
   > &RigUnit_OffsetTransformForItem(
    Item=(Type=Bone,Name="dyn_hair_bk_r_01")
    OffsetTransform=(Rotation=(X=0,Y=0,Z=0,W=1),Translation=(X=-1.6093,Y=0.4097,Z=0.6358),Scale3D=(X=1,Y=1,Z=1))
    bPropagateToChildren=false
    Weight=@n1.Value
   )
   > &RigUnit_ModifyTransforms(
    ItemToModify="((Item=(Type=Bone,Name=\"dyn_hair_mid_l_01\"),Transform=(Rotation=(X=0,Y=0.1736,Z=0,W=0.9848),Translation=(X=0,Y=0,Z=0),Scale3D=(X=1,Y=1,Z=1))),(Item=(Type=Bone,Name=\"dyn_hair_bk_l_01\"),Transform=(Rotation=(X=0,Y=0.1736,Z=0,W=0.9848),Translation=(X=0,Y=0,Z=0 ...(truncated)"
    WeightMinimum=0
    Weight=&RigVMFunction_MathDoubleMul(B=RigVMVariableNode() {variable=Head_Up}.Value, A=RigVMVariableNode() {variable=Head_L}.Value).Result
   )
   n2 = &RigVMFunction_AlphaInterp(
    InterpSpeedDecreasing=5
    InterpSpeedIncreasing=5
    Scale=1
    bClampResult=False
    Value=&RigVMFunction_MathDoubleMul(A=RigVMVariableNode() {variable=Head_L}.Value, B=RigVMVariableNode() {variable=isAiming}.Value).Result
   )
   > &RigUnit_ModifyTransforms(
    ItemToModify="((Item=(Type=Bone,Name=\"dyn_hair_bk_mid_parent\"),Transform=(Rotation=(X=0,Y=-0.1736,Z=0,W=0.9848),Translation=(X=0,Y=0,Z=0),Scale3D=(X=1,Y=1,Z=1))),(Item=(Type=Bone,Name=\"dyn_hair_bk_r_parent\"),Transform=(Rotation=(X=0,Y=-0.0872,Z=0,W=0.9962),Translation=( ...(truncated)"
    WeightMinimum=0
    Weight=@n2.Result
   )
   ItemArray_1 = RigVMDispatchNode(Value=((Type=Bone,Name="dyn_hair_bk_l_parent"),(Type=Bone,Name="dyn_hair_mid_l_parent"),(Type=Bone,Name="dyn_hair_bk_mid_parent"),(Type=Bone,Name="dyn_hair_bk_r_parent"))) {template="DISPATCH_RigVMDispatch_Constant(io Value)"}
   > For_Each_1 = RigVMDispatchNode(Array=@ItemArray_1.Value) {template="DISPATCH_RigVMDispatch_ArrayIterator(in Array,out Element,out Index,out Count,out Ratio)"}
   n3 = &RigVMFunction_AlphaInterp(InterpSpeedDecreasing=10, InterpSpeedIncreasing=10, Scale=1, bClampResult=False, Value=RigVMVariableNode() {variable=isCrouching}.Value)
   > &RigUnit_ModifyTransforms(
    ItemToModify=((Item=(Type=Bone,Name="None"),Transform=(Rotation=(X=0,Y=0,Z=0,W=1),Translation=(X=0,Y=0,Z=0),Scale3D=(X=0.85,Y=1,Z=1))))
    WeightMinimum=0
    Weight=@n3.Result
    ItemToModify.0.Item=@For_Each_1.Element
   )
   n4 = RigVMVariableNode() {variable=Head_R}
   &RigUnit_ModifyTransforms(
    ItemToModify="((Item=(Type=Bone,Name=\"dyn_hair_bk_r_parent\"),Transform=(Rotation=(X=0,Y=0.1305,Z=0,W=0.9914),Translation=(X=0,Y=-2,Z=1),Scale3D=(X=1,Y=1,Z=1))),(Item=(Type=Bone,Name=\"dyn_hair_bk_r_01\"),Transform=(Rotation=(X=0,Y=0,Z=0,W=1),Translation=(X=0,Y=1,Z=1),Scal ...(truncated)"
    WeightMinimum=0
    Weight=@n4.Value
    ExecutePin=@RigVMFunction_Sequence.C
   )
   > &RigUnit_ModifyTransforms(
    ItemToModify="((Item=(Type=Bone,Name=\"dyn_hair_bk_mid_parent\"),Transform=(Rotation=(X=0,Y=-0.1736,Z=0,W=0.9848),Translation=(X=0,Y=1,Z=1),Scale3D=(X=1,Y=1,Z=1))),(Item=(Type=Bone,Name=\"dyn_hair_bk_r_01\"),Transform=(Rotation=(X=0,Y=0,Z=0,W=1),Translation=(X=0,Y=1,Z=1),Sc ...(truncated)"
    WeightMinimum=0
    Weight=@n4.Value
   )
   > &RigUnit_ModifyTransforms(
    ItemToModify=((Item=(Type=Bone,Name="dyn_hair_01"),Transform=(Rotation=(X=0,Y=0,Z=0,W=1),Translation=(X=0,Y=5,Z=-5),Scale3D=(X=1,Y=1,Z=1))),(Item=(Type=Bone,Name="dyn_hair_02"),Transform=(Rotation=(X=0,Y=0,Z=0,W=1),Translation=(X=0,Y=0,Z=0),Scale3D=(X=1,Y=1,Z=1))))
    WeightMinimum=0
    Weight=@n4.Value
   )
   > &RigUnit_OffsetTransformForItem(
    Item=(Type=Bone,Name="dyn_hair_bk_r_02")
    OffsetTransform=(Rotation=(X=0,Y=0,Z=0,W=-1),Translation=(X=-2.2359,Y=0.7696,Z=1.345),Scale3D=(X=1,Y=1,Z=1))
    bPropagateToChildren=false
    Weight=@n4.Value
   )
   > &RigUnit_OffsetTransformForItem(
    Item=(Type=Bone,Name="dyn_hair_bk_r_01")
    OffsetTransform=(Rotation=(X=0,Y=0,Z=0,W=1),Translation=(X=-0.4461,Y=0.3759,Z=0.7775),Scale3D=(X=1,Y=1,Z=1))
    bPropagateToChildren=false
    Weight=@n4.Value
   )
   > &RigUnit_OffsetTransformForItem(
    Item=(Type=Bone,Name="dyn_hair_bk_r_parent")
    OffsetTransform=(Rotation=(X=0,Y=0,Z=0,W=1),Translation=(X=0,Y=0,Z=0),Scale3D=(X=1,Y=1,Z=1))
    Weight=1
    bPropagateToChildren=false
   )
   > &RigUnit_ModifyTransforms(
    ItemToModify=((Item=(Type=Bone,Name="dyn_hair_02"),Transform=(Rotation=(X=0,Y=0,Z=0,W=1),Translation=(X=0,Y=3,Z=0),Scale3D=(X=1,Y=1,Z=1))),(Item=(Type=Bone,Name="dyn_hair_01"),Transform=(Rotation=(X=0,Y=0,Z=0,W=1),Translation=(X=3,Y=3,Z=0),Scale3D=(X=1,Y=1,Z=1))))
    WeightMinimum=0
    Weight=RigVMVariableNode() {variable=HeadDown}.Value
   )
   > &RigUnit_ModifyTransforms(
    ItemToModify="((Item=(Type=Bone,Name=\"dyn_hair_02\"),Transform=(Rotation=(X=0,Y=-0.1736,Z=0,W=0.9848),Translation=(X=0,Y=0,Z=0),Scale3D=(X=1,Y=1,Z=1))),(Item=(Type=Bone,Name=\"dyn_hair_04\"),Transform=(Rotation=(X=0,Y=-0.1736,Z=0,W=0.9848),Translation=(X=0,Y=0,Z=0),Scale3D ...(truncated)"
    WeightMinimum=0
    Weight=&RigVMFunction_MathDoubleMul(B=RigVMVariableNode() {variable=isAiming}.Value, A=RigVMVariableNode() {variable=Head_R}.Value).Result
   )
   > &RigUnit_ModifyTransforms(
    ItemToModify="((Item=(Type=Bone,Name=\"dyn_hair_bk_mid_parent\"),Transform=(Rotation=(X=0,Y=-0.1736,Z=0,W=0.9848),Translation=(X=0,Y=0,Z=0),Scale3D=(X=1,Y=1,Z=1))),(Item=(Type=Bone,Name=\"dyn_hair_mid_l_parent\"),Transform=(Rotation=(X=0,Y=-0.2588,Z=0,W=0.9659),Translation= ...(truncated)"
    WeightMinimum=0
    Weight=RigVMVariableNode() {variable=isSkydiving}.Value
   )
   &RigUnit_ModifyTransforms(
    ItemToModify="((Item=(Type=Bone,Name=\"dyn_hair_bk_l_01\"),Transform=(Rotation=(X=0,Y=0.0872,Z=0,W=0.9962),Translation=(X=0,Y=0,Z=0),Scale3D=(X=1,Y=1,Z=1))),(Item=(Type=Bone,Name=\"dyn_hair_bk_mid_01\"),Transform=(Rotation=(X=0,Y=0.0436,Z=0,W=0.999),Translation=(X=0,Y=0,Z=0 ...(truncated)"
    Weight=1
    WeightMinimum=0
    ExecutePin=@RigVMFunction_Sequence.A
   )
   Subtract = &RigVMFunction_MathDoubleSub(A=1, B=RigVMVariableNode() {variable=Head_Up}.Value)
   Absolute = &RigVMFunction_MathDoubleAbs(Value=@Subtract.Result)
   Multiply = &RigVMFunction_MathDoubleMul(B=@Absolute.Result, A=RigVMVariableNode() {variable=isAiming}.Value)
   AlphaInterp = &RigVMFunction_AlphaInterp(InterpSpeedDecreasing=5, InterpSpeedIncreasing=5, Scale=1, bClampResult=False, Value=@Multiply.Result)
   > &RigUnit_ModifyTransforms(
    ItemToModify="((Item=(Type=Bone,Name=\"None\"),Transform=(Rotation=(X=0,Y=0,Z=0,W=1),Translation=(X=0,Y=0,Z=0),Scale3D=(X=1,Y=1,Z=1))),(Item=(Type=Bone,Name=\"dyn_hair_bk_mid_parent\"),Transform=(Rotation=(X=0,Y=0,Z=0,W=1),Translation=(X=0,Y=0,Z=0),Scale3D=(X=1,Y=1,Z=1))),( ...(truncated)"
    WeightMinimum=0
    Weight=@AlphaInterp.Result
   )
   ItemArray = RigVMDispatchNode(Value=((Type=Bone,Name="dyn_hair_bk_l_parent"),(Type=Bone,Name="dyn_hair_mid_l_parent"),(Type=Bone,Name="dyn_hair_bk_mid_parent"),(Type=Bone,Name="dyn_hair_bk_r_parent"))) {template="DISPATCH_RigVMDispatch_Constant(io Value)"}
   > For_Each = RigVMDispatchNode(Array=@ItemArray.Value) {template="DISPATCH_RigVMDispatch_ArrayIterator(in Array,out Element,out Index,out Count,out Ratio)"}
   > &RigUnit_ModifyTransforms(
    ItemToModify=((Item=(Type=Bone,Name="dyn_hair_bk_r_parent"),Transform=(Rotation=(X=0,Y=0.0872,Z=0,W=0.9962),Translation=(X=0,Y=0,Z=0),Scale3D=(X=0.7,Y=1,Z=0.5))))
    WeightMinimum=0
    ItemToModify.0.Item=@For_Each.Element
    Weight=RigVMVariableNode() {variable=isTacticalJumping}.Value
   )
   n5 = &RigVMFunction_MathDoubleSub(A=1, B=RigVMVariableNode() {variable=isAiming}.Value)
   n6 = &RigVMFunction_AlphaInterp(
    InterpSpeedDecreasing=5
    InterpSpeedIncreasing=5
    Scale=1
    bClampResult=False
    Value=&RigVMFunction_MathDoubleMul(B=&RigVMFunction_MathDoubleAbs(Value=@n5.Result).Result, A=RigVMVariableNode() {variable=isJumping}.Value).Result
   )
   > &RigUnit_ModifyTransforms(
    ItemToModify=((Item=(Type=Bone,Name="None"),Transform=(Rotation=(X=0,Y=0,Z=0,W=1),Translation=(X=0,Y=3,Z=0),Scale3D=(X=0.8,Y=1,Z=1))))
    WeightMinimum=0
    Weight=@n6.Result
    ItemToModify.0.Item=@For_Each.Element
   )
   n7 = RigVMVariableNode() {variable=Head_Up}
   &RigUnit_ModifyTransforms(
    ItemToModify="((Item=(Type=Bone,Name=\"dyn_hair_bk_l_01\"),Transform=(Rotation=(X=0,Y=0,Z=0,W=1),Translation=(X=2,Y=0,Z=0),Scale3D=(X=1,Y=1,Z=1))),(Item=(Type=Bone,Name=\"dyn_hair_bk_l_parent\"),Transform=(Rotation=(X=0,Y=-0.0872,Z=0,W=0.9962),Translation=(X=0,Y=0,Z=0),Scal ...(truncated)"
    WeightMinimum=0
    Weight=@n7.Value
    ExecutePin=@RigVMFunction_Sequence.D
   )
   > &RigUnit_ModifyTransforms(
    ItemToModify="((Item=(Type=Bone,Name=\"dyn_hair_bk_mid_parent\"),Transform=(Rotation=(X=0,Y=-0.0872,Z=0,W=0.9962),Translation=(X=0,Y=0,Z=0),Scale3D=(X=1,Y=1,Z=1))),(Item=(Type=Bone,Name=\"dyn_hair_bk_mid_01\"),Transform=(Rotation=(X=0,Y=0,Z=0,W=1),Translation=(X=2,Y=0,Z=0), ...(truncated)"
    WeightMinimum=0
    Weight=@n7.Value
   )
   n8 = &RigVMFunction_MathDoubleSub(A=1, B=RigVMVariableNode() {variable=isAiming}.Value)
   n9 = &RigVMFunction_AlphaInterp(
    InterpSpeedDecreasing=5
    InterpSpeedIncreasing=5
    Scale=1
    bClampResult=False
    Value=&RigVMFunction_MathDoubleMul(B=&RigVMFunction_MathDoubleAbs(Value=@n8.Result).Result, A=@n7.Value).Result
   )
   > &RigUnit_ModifyTransforms(
    ItemToModify="((Item=(Type=Bone,Name=\"dyn_hair_02\"),Transform=(Rotation=(X=0,Y=0.0436,Z=0,W=0.999),Translation=(X=-2,Y=2,Z=0),Scale3D=(X=1,Y=1,Z=1))),(Item=(Type=Bone,Name=\"dyn_hair_01\"),Transform=(Rotation=(X=0.0076,Y=0.0868,Z=-0.0868,W=0.9924),Translation=(X=-2,Y=0,Z= ...(truncated)"
    WeightMinimum=0
    Weight=@n9.Result
   )
   n10 = &RigUnit_SphericalPoseReader(
    ActiveRegionScaleFactors=(PositiveWidth=1,NegativeWidth=1,PositiveHeight=1,NegativeHeight=1)
    ActiveRegionSize=0.18
    Debug=(bDrawDebug=true,bDraw2D=False,bDrawLocalAxes=False,DebugScale=25,DebugSegments=20,DebugThickness=0.25)
    DriverAxis=(X=1,Y=0,Z=0)
    DriverItem=(Type=Bone,Name="upperarm_l")
    FalloffRegionScaleFactors=(PositiveWidth=1,NegativeWidth=1,PositiveHeight=1,NegativeHeight=1)
    FalloffSize=0.25
    FlipHeightScaling=False
    FlipWidthScaling=False
    OptionalParentItem=(Type=Bone,Name="clavicle_l")
    RotationOffset=(X=110,Y=10,Z=0)
    ExecutePin=@RigVMFunction_Sequence.E
   )
   n11 = &RigVMFunction_AlphaInterp(InterpSpeedDecreasing=5, InterpSpeedIncreasing=5, Scale=1, bClampResult=False, Value=@n10.OutputParam)
   > &RigUnit_ModifyTransforms(
    ItemToModify="((Item=(Type=Bone,Name=\"dyn_hair_01\"),Transform=(Rotation=(X=0,Y=0.0872,Z=0,W=0.9962),Translation=(X=0,Y=0,Z=0),Scale3D=(X=1,Y=1,Z=1))),(Item=(Type=Bone,Name=\"dyn_hair_04\"),Transform=(Rotation=(X=0,Y=0.1736,Z=0,W=0.9848),Translation=(X=0,Y=0,Z=0),Scale3D=( ...(truncated)"
    WeightMinimum=0
    Weight=@n11.Result
   )
   n12 = &RigVMFunction_AlphaInterp(
    InterpSpeedDecreasing=10
    InterpSpeedIncreasing=10
    Scale=1
    bClampResult=False
    Value=&RigVMFunction_MathDoubleMul(A=@n10.OutputParam, B=RigVMVariableNode() {variable=isEmoting}.Value).Result
   )
   > &RigUnit_ModifyTransforms(
    ItemToModify="((Item=(Type=Bone,Name=\"dyn_hair_01\"),Transform=(Rotation=(X=-0.0015,Y=0.0871,Z=0.0174,W=0.996),Translation=(X=0,Y=0,Z=0),Scale3D=(X=1,Y=1,Z=1))),(Item=(Type=Bone,Name=\"dyn_hair_03\"),Transform=(Rotation=(X=-0.0046,Y=0.0521,Z=0.087,W=0.9948),Translation=(X= ...(truncated)"
    WeightMinimum=0
    Weight=@n12.Result
   )
   > &RigUnit_ModifyTransforms(
    ItemToModify="((Item=(Type=Bone,Name=\"dyn_hair_01\"),Transform=(Rotation=(X=0.0094,Y=-0.2162,Z=0.0426,W=0.9754),Translation=(X=0,Y=0,Z=0),Scale3D=(X=1,Y=1,Z=1))),(Item=(Type=Bone,Name=\"dyn_hair_03\"),Transform=(Rotation=(X=0,Y=0.0523,Z=0,W=0.9986),Translation=(X=0,Y=0,Z=0 ...(truncated)"
    WeightMinimum=0
    Weight=&RigVMFunction_MathDoubleMul(B=@n11.Result, A=RigVMVariableNode() {variable=Head_Up}.Value).Result
   )
   n13 = RigVMDispatchNode(False=0, True=1, Condition=RigVMVariableNode() {variable=FrontEnd}.Value) {template="DISPATCH_RigVMDispatch_If(in Condition,in True,in False,out Result)"}
   > &RigUnit_OffsetTransformForItem(
    Item=(Type=Bone,Name="dyn_hair_bk_r_parent")
    OffsetTransform=(Rotation=(X=0,Y=0,Z=0,W=1),Translation=(X=0,Y=0,Z=0),Scale3D=(X=1,Y=1,Z=1))
    bPropagateToChildren=false
    Weight=@n13.Result
   )
   > &RigUnit_OffsetTransformForItem(
    Item=(Type=Bone,Name="dyn_hair_bk_mid_01")
    OffsetTransform=(Rotation=(X=0.0431,Y=-0.2576,Z=-0.107,W=0.9593),Translation=(X=-2.8768,Y=-0.0003,Z=0.0001),Scale3D=(X=1,Y=1,Z=1))
    bPropagateToChildren=false
    Weight=@n13.Result
   )
   > &RigUnit_OffsetTransformForItem(
    Item=(Type=Bone,Name="dyn_hair_bk_r_01")
    OffsetTransform=(Rotation=(X=0.0329,Y=-0.148,Z=-0.1809,W=0.9718),Translation=(X=-1.029,Y=-3.4102,Z=1.4498),Scale3D=(X=0.7244,Y=1,Z=1))
    bPropagateToChildren=true
    Weight=@n13.Result
   )
   > &RigUnit_OffsetTransformForItem(
    Item=(Type=Bone,Name="dyn_hair_03")
    OffsetTransform=(Rotation=(X=0,Y=0,Z=-0.0714,W=0.9974),Translation=(X=0,Y=0,Z=0),Scale3D=(X=1,Y=1,Z=1))
    bPropagateToChildren=true
    Weight=@n13.Result
   )
   > &RigUnit_OffsetTransformForItem(
    Item=(Type=Bone,Name="dyn_hair_04")
    OffsetTransform=(Rotation=(X=0.0009,Y=-0.0851,Z=-0.0959,W=-0.9917),Translation=(X=0,Y=0,Z=0),Scale3D=(X=1,Y=1,Z=1))
    bPropagateToChildren=true
    Weight=@n13.Result
   )
   > &RigUnit_OffsetTransformForItem(
    Item=(Type=Bone,Name="dyn_hair_bk_02")
    OffsetTransform=(Rotation=(X=0.0726,Y=0.3763,Z=0.1293,W=-0.9146),Translation=(X=-0.0005,Y=-0.0004,Z=4.881),Scale3D=(X=1,Y=1,Z=1))
    bPropagateToChildren=true
    Weight=@n13.Result
   )
   > &RigUnit_OffsetTransformForItem(
    Item=(Type=Bone,Name="dyn_hair_bk_l_01")
    OffsetTransform=(Rotation=(X=0,Y=0,Z=-0.2069,W=0.9784),Translation=(X=0,Y=0,Z=0),Scale3D=(X=1,Y=1,Z=1))
    bPropagateToChildren=true
    Weight=@n13.Result
   )
   > &RigUnit_OffsetTransformForItem(
    Item=(Type=Bone,Name="dyn_hair_mid_l_01")
    OffsetTransform=(Rotation=(X=0,Y=0,Z=0,W=-1),Translation=(X=-2.4023,Y=0,Z=0),Scale3D=(X=1,Y=1,Z=1))
    bPropagateToChildren=true
    Weight=@n13.Result
   )
  ]}
]
hierarchy [
 root {loc=<0 0 0>, rot=<0 0 0 1>}
 pelvis {parent=root, loc=<0 -0.5301 98.0571>, rot=<0.0127 0.707 -0.0127 -0.707>}
 spine_01 {parent=pelvis, loc=<6.1208 -0.9328 0>, rot=<0 0 0.034 -0.9994>}
 spine_02 {parent=spine_01, loc=<6.5997 -0.2095 0>, rot=<0 0 0.0238 -0.9997>}
 spine_03 {parent=spine_02, loc=<6.4914 -0.1237 0>, rot=<0 0 0.0149 -0.9999>}
 spine_04 {parent=spine_03, loc=<6.9598 0.0794 0>, rot=<0 0 -0.0132 -0.9999>}
 spine_05 {parent=spine_04, loc=<6.9987 0 0>, rot=<0 0 -0.0015 -1>}
 neck_01 {parent=spine_05, loc=<11.3138 0.1529 0>, rot=<0 0 -0.2003 -0.9797>}
 neck_02 {parent=neck_01, loc=<4.5627 0 0>, rot=<0 0 0.0504 -0.9987>}
 head {parent=neck_02, loc=<5.3914 0 0>, rot=<0 0 0.115 -0.9934>}
 dyn_hair_bk_r_base {parent=head, loc=<2.6263 -3.853 5.867>, rot=<0.0743 -0.424 0.8922 -0.1369>}
 dyn_hair_bk_r_parent {parent=dyn_hair_bk_r_base, loc=<0 0 0>, rot=<0 0 0 1>}
 dyn_hair_bk_r_01 {parent=dyn_hair_bk_r_parent, loc=<6.0461 -0.5909 -0.864>, rot=<0.1362 0.0809 -0.1117 -0.981>}
 dyn_hair_bk_r_02 {parent=dyn_hair_bk_r_01, loc=<6.9335 0 0>, rot=<-0.0698 0 0 -0.9976>}
 dyn_hair_bk_mid_base {parent=head, loc=<2.2224 -6.1632 2.7259>, rot=<0.0719 -0.579 0.8121 -0.0088>}
 dyn_hair_bk_mid_parent {parent=dyn_hair_bk_mid_base, loc=<0 0 0>, rot=<0 0 0 1>}
 dyn_hair_bk_mid_01 {parent=dyn_hair_bk_mid_parent, loc=<5.5169 0 0>, rot=<0.258 0.2147 -0.1263 -0.9335>}
 dyn_hair_bk_02 {parent=dyn_hair_bk_mid_01, loc=<6.0179 0 0>, rot=<-0.2172 -0.0977 0.1049 -0.9655>}
 dyn_hair_bk_l_base {parent=head, loc=<2.2538 -6.2987 -1.8152>, rot=<0.0032 -0.766 0.6428 -0.0027>}
 dyn_hair_bk_l_parent {parent=dyn_hair_bk_l_base, loc=<0 0 0>, rot=<0 0 0 1>}
 dyn_hair_bk_l_01 {parent=dyn_hair_bk_l_parent, loc=<5.5131 0.5016 1.1859>, rot=<0.1017 0.2417 -0.0255 -0.9647>}
 dyn_hair_bk_l_02 {parent=dyn_hair_bk_l_01, loc=<7.0897 0 0>, rot=<0 0 0 1>}
 dyn_hair_mid_l_base {parent=head, loc=<2.6436 -3.6767 -6.2618>, rot=<0.0893 -0.9397 0.3068 -0.1216>}
 dyn_hair_mid_l_parent {parent=dyn_hair_mid_l_base, loc=<0 0 0>, rot=<0 0 0 1>}
 dyn_hair_mid_l_01 {parent=dyn_hair_mid_l_parent, loc=<5.4213 0 0>, rot=<-0.0387 0.0027 -0.0241 -0.999>}
 dyn_hair_mid_l_02 {parent=dyn_hair_mid_l_01, loc=<3.8324 0 0>, rot=<0.0523 0 0 -0.9986>}
 dyn_hair_base {parent=head, loc=<2.52 5.7462 -6.4959>, rot=<0.0304 -0.9611 0.0073 -0.2744>}
 dyn_hair_parent {parent=dyn_hair_base, loc=<0 0 0>, rot=<0 0 0 1>}
 dyn_hair_01 {parent=dyn_hair_parent, loc=<4.7947 -0.2205 -0.8879>, rot=<0.0728 -0.267 0.0654 -0.9587>}
 dyn_hair_02 {parent=dyn_hair_01, loc=<4.8923 0 0>, rot=<0.0845 -0.2168 -0.2382 -0.943>}
 dyn_hair_03 {parent=dyn_hair_02, loc=<5.5595 0 0>, rot=<0.1579 0.2951 -0.123 -0.9343>}
 dyn_hair_04 {parent=dyn_hair_03, loc=<6.6748 0 0>, rot=<-0.1626 0.1218 0.2021 -0.9581>}
 dyn_hair_05 {parent=dyn_hair_04, loc=<3.7419 0 0>, rot=<-0.1509 -0.0248 -0.065 -0.9861>}
 dyn_hair_06 {parent=dyn_hair_05, loc=<4.1463 -2.3878 -0.6041>, rot=<0.173 -0.0436 0.0755 -0.9811>}
 dyn_hair_07 {parent=dyn_hair_05, loc=<4.2014 -2.3509 -4.5793>, rot=<0.1779 -0.0473 0.1256 -0.9749>}
 dyn_hair_08 {parent=dyn_hair_05, loc=<3.7845 -2.4255 1.9131>, rot=<0.1062 0.1978 0.239 -0.9447>}
 clavicle_l {parent=spine_05, loc=<5.4831 -0.1053 -3.2946>, rot=<0.0465 -0.7289 -0.009 -0.6829>}
 upperarm_l {parent=clavicle_l, loc=<9.8658 -1.2136 -0.4105>, rot=<0.0164 -0.3245 -0.0269 -0.9454>}
 lowerarm_l {parent=upperarm_l, loc=<26.465 -0.8345 0.259>, rot=<0.0036 0.0161 -0.2298 -0.9731>}
 clavicle_r {parent=spine_05, loc=<5.4831 -0.1053 3.2946>, rot=<0.7289 0.0465 -0.6829 0.009>}
 upperarm_r {parent=clavicle_r, loc=<-9.8658 1.2136 0.4105>, rot=<0.0164 -0.3245 -0.0269 -0.9454>}
 lowerarm_r {parent=upperarm_r, loc=<-26.465 0.8345 -0.259>, rot=<0.0036 0.0161 -0.2298 -0.9731>}
]
variables [
 FrontEnd {type=bool, default=False}
 HeadDown {type=double, default=0}
 Head_L {type=double, default=0}
 Head_R {type=double, default=0}
 Head_Up {type=double, default=0}
 isAiming {type=double, default=0}
 isCrouching {type=double, default=0}
 isEmoting {type=double, default=0}
 isJumping {type=double, default=0}
 isSkydiving {type=double, default=0}
 isTacticalJumping {type=double, default=0}
]
