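# Animation Blueprint for the F_MED_QuarterClaspJoin face accessory, animating its hair bones as secondary motion.
# The pose copied from the attached parent passes through a Control Rig, weighted bone constraints to head and upper body, then sprung AnimDynamics nodes.
# A post Control Rig takes front-end, head-rotation and movement-state inputs, mostly computed each update.
# The update also stores parameter-asset gravity and rigid-body settings, which the AnimGraph never reads.
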AnimBlueprint /BRCosmetics/Characters/Player/Female/Medium/Bodies/F_MED_QuarterClaspJoin/Meshes/Parts/F_MED_QuarterClaspJoin_FaceAcc_AnimBP
target_skeleton=&/F_MED_QuarterClaspJoin_FaceAcc_Skeleton
graphs [
 AnimGraph {nodes [
   n1 = Control Rig(Source=Copy Pose From Mesh(SourceMeshComponent=Get Skeletal_Mesh() -> (Skeletal_Mesh:Skeletal Mesh Component Object Reference)) {bUseAttachedParent=true}) {controlRigAssetReference {blueprintRigClass=&/F_MED_QuarterClaspJoin_FaceAcc_CtrlRig.F_MED_QuarterClaspJoin_FaceAcc_CtrlRig_C}}
   n2 = "Constraint - dyn_hair_mid_l_parent (P)"!(ComponentPose=Local To Component(LocalPose=@n1), Alpha=1, ConstraintWeights_0=1, ConstraintWeights_1:float=0) -> (Pose) {boneToModify {boneName=dyn_hair_mid_l_parent}, constraintSetup [{targetBone {boneName=head}}], constraintWeights [1,0]}
   n3 = "Constraint - dyn_hair_mid_l_01 (PP)"!(ComponentPose=@n2, Alpha=1, ConstraintWeights_0=0.8, ConstraintWeights_1:float=0.2) -> (Pose) {
    boneToModify {boneName=dyn_hair_mid_l_01}
    constraintSetup [{targetBone {boneName=clavicle_l}}, {targetBone {boneName=head}}]
    constraintWeights [1,1]
   }
   n4 = "Constraint - dyn_hair_mid_l_02 (PPP)"!(ComponentPose=@n3, Alpha=1, ConstraintWeights_0=0.8521, ConstraintWeights_1:float=0.1479, ConstraintWeights_2:float=0) -> (Pose) {
    boneToModify {boneName=dyn_hair_mid_l_02}
    constraintSetup [{targetBone {boneName=spine_05}}, {targetBone {boneName=clavicle_l}}, {targetBone {boneName=neck_02}}]
    constraintWeights [1,1,1]
   }
   n5 = "Constraint - dyn_hair_bk_r_parent (P)"!(ComponentPose=@n4, Alpha=1, ConstraintWeights_0=1, ConstraintWeights_1:float=0) -> (Pose) {boneToModify {boneName=dyn_hair_bk_r_parent}, constraintSetup [{targetBone {boneName=head}}], constraintWeights [1,0]}
   n6 = "Constraint - dyn_hair_bk_r_01 (PPP)"!(ComponentPose=@n5, Alpha=1, ConstraintWeights_0=0.2, ConstraintWeights_1:float=0.6, ConstraintWeights_2:float=0.2) -> (Pose) {
    boneToModify {boneName=dyn_hair_bk_r_01}
    constraintSetup [{targetBone {boneName=clavicle_r}}, {targetBone {boneName=spine_05}}, {targetBone {boneName=head}}]
    constraintWeights [1,1,1]
   }
   n7 = "Constraint - dyn_hair_bk_r_02 (PP)"!(ComponentPose=@n6, Alpha=1, ConstraintWeights_0=0.7421, ConstraintWeights_1:float=0.2579) -> (Pose) {
    boneToModify {boneName=dyn_hair_bk_r_02}
    constraintSetup [{targetBone {boneName=spine_05}}, {targetBone {boneName=clavicle_r}}]
    constraintWeights [1,1]
   }
   n8 = "Constraint - dyn_hair_bk_mid_parent (P)"!(ComponentPose=@n7, Alpha=1, ConstraintWeights_0=1, ConstraintWeights_1:float=0) -> (Pose) {boneToModify {boneName=dyn_hair_bk_mid_parent}, constraintSetup [{targetBone {boneName=head}}], constraintWeights [1,0]}
   n9 = "Constraint - dyn_hair_bk_mid_01 (PP)"!(ComponentPose=@n8, Alpha=1, ConstraintWeights_0=0.8, ConstraintWeights_1:float=0.2) -> (Pose) {
    boneToModify {boneName=dyn_hair_bk_mid_01}
    constraintSetup [{targetBone {boneName=spine_05}}, {targetBone {boneName=head}}]
    constraintWeights [1,1]
   }
   n10 = "Constraint - dyn_hair_bk_02 (PPP)"!(ComponentPose=@n9, Alpha=1, ConstraintWeights_0=0.9617, ConstraintWeights_1:float=0.0201, ConstraintWeights_2:float=0.0181) -> (Pose) {
    boneToModify {boneName=dyn_hair_bk_02}
    constraintSetup [{targetBone {boneName=spine_05}}, {targetBone {boneName=clavicle_r}}, {targetBone {boneName=clavicle_l}}]
    constraintWeights [1,1,1]
   }
   n11 = "Constraint - dyn_hair_bk_l_parent (P)"!(ComponentPose=@n10, Alpha=1, ConstraintWeights_0=1, ConstraintWeights_1:float=0) -> (Pose) {boneToModify {boneName=dyn_hair_bk_l_parent}, constraintSetup [{targetBone {boneName=head}}], constraintWeights [1,0]}
   n12 = "Constraint - dyn_hair_bk_l_01 (PPP)"!(ComponentPose=@n11, Alpha=1, ConstraintWeights_0=0.4, ConstraintWeights_1:float=0.2, ConstraintWeights_2:float=0.4) -> (Pose) {
    boneToModify {boneName=dyn_hair_bk_l_01}
    constraintSetup [{targetBone {boneName=clavicle_l}}, {targetBone {boneName=head}}, {targetBone {boneName=spine_05}}]
    constraintWeights [1,1,1]
   }
   n13 = "Constraint - dyn_hair_bk_l_02 (PP)"!(ComponentPose=@n12, Alpha=1, ConstraintWeights_0=0.8384, ConstraintWeights_1:float=0.1616) -> (Pose) {
    boneToModify {boneName=dyn_hair_bk_l_02}
    constraintSetup [{targetBone {boneName=spine_05}}, {targetBone {boneName=clavicle_l}}]
    constraintWeights [1,1]
   }
   n14 = Anim Dynamics - Bone: dyn_hair_mid_l_parent!(ComponentPose=@n13, Alpha=1, GravityOverride:Vector="0,0,-980", ExternalForce="0,0,0") -> (Pose) {
    angularDampingOverride=0.9999
    angularSpringConstant=5000
    bAngularSpring=true
    bOverrideAngularDamping=true
    bUseGravityOverride=true
    gravityScale=0.35
    simulationSpace=RootRelative
    boundBone {boneName=dyn_hair_mid_l_parent}
    componentLinearAccScale=<1 1 1>
    componentLinearVelScale=<1 1 1>
    gravityOverride {z=-980}
    physicsBodyDefinitions [{
      boundBone {boneName=dyn_hair_mid_l_parent}
      boxExtents=<40 40 40>
      constraintSetup {angularLimitsMax=<0 -1 3>, angularLimitsMin=<0 -5 -3>, angularTarget=<1 0 0>}
      localJointOffset=<4 0 0>
     }]
   }
   n15 = Anim Dynamics - Bone: dyn_hair_bk_l_parent!(ComponentPose=@n14, Alpha=1, GravityOverride:Vector="0,0,-980", ExternalForce="0,0,0") -> (Pose) {
    angularDampingOverride=0.9999
    angularSpringConstant=5000
    bAngularSpring=true
    bOverrideAngularDamping=true
    bUseGravityOverride=true
    gravityScale=0.35
    simulationSpace=RootRelative
    boundBone {boneName=dyn_hair_bk_l_parent}
    componentLinearAccScale=<1 1 1>
    componentLinearVelScale=<1 1 1>
    gravityOverride {z=-980}
    physicsBodyDefinitions [{
      boundBone {boneName=dyn_hair_bk_l_parent}
      boxExtents=<40 40 40>
      constraintSetup {angularLimitsMax=<0 -1 3>, angularLimitsMin=<0 -5 -3>, angularTarget=<1 0 0>}
      localJointOffset=<4 0 0>
     }]
   }
   n16 = Anim Dynamics - Bone: dyn_hair_bk_mid_parent!(ComponentPose=@n15, Alpha=1, GravityOverride:Vector="0,0,-980", ExternalForce="0,0,0") -> (Pose) {
    angularDampingOverride=0.9999
    angularSpringConstant=5000
    bAngularSpring=true
    bOverrideAngularDamping=true
    bUseGravityOverride=true
    gravityScale=0.35
    simulationSpace=RootRelative
    boundBone {boneName=dyn_hair_bk_mid_parent}
    componentLinearAccScale=<1 1 1>
    componentLinearVelScale=<1 1 1>
    gravityOverride {z=-980}
    physicsBodyDefinitions [{
      boundBone {boneName=dyn_hair_bk_mid_parent}
      boxExtents=<40 40 40>
      constraintSetup {angularLimitsMax=<0 -1 3>, angularLimitsMin=<0 -5 -3>, angularTarget=<1 0 0>}
      localJointOffset=<4 0 0>
     }]
   }
   n17 = Anim Dynamics - Bone: dyn_hair_bk_r_parent!(ComponentPose=@n16, Alpha=1, GravityOverride:Vector="0,0,-980", ExternalForce="0,0,0") -> (Pose) {
    angularDampingOverride=0.9999
    angularSpringConstant=5000
    bAngularSpring=true
    bOverrideAngularDamping=true
    bUseGravityOverride=true
    gravityScale=0.35
    simulationSpace=RootRelative
    boundBone {boneName=dyn_hair_bk_r_parent}
    componentLinearAccScale=<1 1 1>
    componentLinearVelScale=<1 1 1>
    gravityOverride {z=-980}
    physicsBodyDefinitions [{
      boundBone {boneName=dyn_hair_bk_r_parent}
      boxExtents=<40 40 40>
      constraintSetup {angularLimitsMax=<0 -1 3>, angularLimitsMin=<0 -5 -3>, angularTarget=<1 0 0>}
      localJointOffset=<4 0 0>
     }]
   }
   n18 = Anim Dynamics - Bone: dyn_hair_mid_l_01!(ComponentPose=@n17, Alpha=1, GravityOverride:Vector="0,0,-980", ExternalForce="0,0,0") -> (Pose) {
    angularDampingOverride=0.9999
    angularSpringConstant=5000
    bAngularSpring=true
    bOverrideAngularDamping=true
    bUseGravityOverride=true
    gravityScale=0.35
    simulationSpace=RootRelative
    boundBone {boneName=dyn_hair_mid_l_01}
    componentLinearAccScale=<1 1 1>
    componentLinearVelScale=<1 1 1>
    gravityOverride {z=-980}
    physicsBodyDefinitions [{
      boundBone {boneName=dyn_hair_mid_l_01}
      boxExtents=<20 30 30>
      constraintSetup {angularLimitsMax=<0 -1 3>, angularLimitsMin=<0 -5 -3>, angularTarget=<1 0 0>}
      localJointOffset=<4 0 0>
     }]
   }
   n19 = Anim Dynamics - Bone: dyn_hair_bk_l_01!(ComponentPose=@n18, Alpha=1, GravityOverride:Vector="0,0,-980", ExternalForce="0,0,0") -> (Pose) {
    angularDampingOverride=0.9999
    angularSpringConstant=5000
    bAngularSpring=true
    bOverrideAngularDamping=true
    bUseGravityOverride=true
    gravityScale=0.35
    simulationSpace=RootRelative
    boundBone {boneName=dyn_hair_bk_l_01}
    componentLinearAccScale=<1 1 1>
    componentLinearVelScale=<1 1 1>
    gravityOverride {z=-980}
    physicsBodyDefinitions [{
      boundBone {boneName=dyn_hair_bk_l_01}
      boxExtents=<20 30 30>
      constraintSetup {angularLimitsMax=<0 -1 3>, angularLimitsMin=<0 -5 -3>, angularTarget=<1 0 0>}
      localJointOffset=<4 0 0>
     }]
   }
   n20 = Anim Dynamics - Bone: dyn_hair_bk_mid_01!(ComponentPose=@n19, Alpha=1, GravityOverride:Vector="0,0,-980", ExternalForce="0,0,0") -> (Pose) {
    angularDampingOverride=0.9999
    angularSpringConstant=5000
    bAngularSpring=true
    bOverrideAngularDamping=true
    bUseGravityOverride=true
    gravityScale=0.35
    simulationSpace=RootRelative
    boundBone {boneName=dyn_hair_bk_mid_01}
    componentLinearAccScale=<1 1 1>
    componentLinearVelScale=<1 1 1>
    gravityOverride {z=-980}
    physicsBodyDefinitions [{
      boundBone {boneName=dyn_hair_bk_mid_01}
      boxExtents=<20 30 30>
      constraintSetup {angularLimitsMax=<0 -1 3>, angularLimitsMin=<0 -5 -3>, angularTarget=<1 0 0>}
      localJointOffset=<4 0 0>
     }]
   }
   n21 = Anim Dynamics - Bone: dyn_hair_bk_r_01!(ComponentPose=@n20, Alpha=1, GravityOverride:Vector="0,0,-980", ExternalForce="0,0,0") -> (Pose) {
    angularDampingOverride=0.9999
    angularSpringConstant=5000
    bAngularSpring=true
    bOverrideAngularDamping=true
    bUseGravityOverride=true
    gravityScale=0.35
    simulationSpace=RootRelative
    boundBone {boneName=dyn_hair_bk_r_01}
    componentLinearAccScale=<1 1 1>
    componentLinearVelScale=<1 1 1>
    gravityOverride {z=-980}
    physicsBodyDefinitions [{
      boundBone {boneName=dyn_hair_bk_r_01}
      boxExtents=<20 30 30>
      constraintSetup {angularLimitsMax=<0 -1 3>, angularLimitsMin=<0 -5 -3>, angularTarget=<1 0 0>}
      localJointOffset=<4 0 0>
     }]
   }
   n22 = Anim Dynamics - Bone: dyn_hair_mid_l_02!(ComponentPose=@n21, Alpha=1, GravityOverride:Vector="0,0,-980", ExternalForce="0,0,0") -> (Pose) {
    angularDampingOverride=0.9999
    angularSpringConstant=5000
    bAngularSpring=true
    bOverrideAngularDamping=true
    bUseGravityOverride=true
    gravityScale=0.35
    simulationSpace=RootRelative
    boundBone {boneName=dyn_hair_mid_l_02}
    componentLinearAccScale=<1 1 1>
    componentLinearVelScale=<1 1 1>
    gravityOverride {z=-980}
    physicsBodyDefinitions [{
      boundBone {boneName=dyn_hair_mid_l_02}
      boxExtents=<30 40 40>
      constraintSetup {angularLimitsMax=<0 -1 0>, angularLimitsMin=<0 -5 0>, angularTarget=<1 0 0>}
      localJointOffset=<4 0 0>
     }]
   }
   n23 = Anim Dynamics - Bone: dyn_hair_bk_l_02!(ComponentPose=@n22, Alpha=1, GravityOverride:Vector="0,0,-980", ExternalForce="0,0,0") -> (Pose) {
    angularDampingOverride=0.9999
    angularSpringConstant=5000
    bAngularSpring=true
    bOverrideAngularDamping=true
    bUseGravityOverride=true
    gravityScale=0.35
    simulationSpace=RootRelative
    boundBone {boneName=dyn_hair_bk_l_02}
    componentLinearAccScale=<1 1 1>
    componentLinearVelScale=<1 1 1>
    gravityOverride {z=-980}
    physicsBodyDefinitions [{
      boundBone {boneName=dyn_hair_bk_l_02}
      boxExtents=<30 40 40>
      constraintSetup {angularLimitsMax=<0 -1 0>, angularLimitsMin=<0 -5 0>, angularTarget=<1 0 0>}
      localJointOffset=<4 0 0>
     }]
   }
   n24 = Anim Dynamics - Bone: dyn_hair_bk_02!(ComponentPose=@n23, Alpha=1, GravityOverride:Vector="0,0,-980", ExternalForce="0,0,0") -> (Pose) {
    angularDampingOverride=0.9999
    angularSpringConstant=5000
    bAngularSpring=true
    bOverrideAngularDamping=true
    bUseGravityOverride=true
    gravityScale=0.35
    simulationSpace=RootRelative
    boundBone {boneName=dyn_hair_bk_02}
    componentLinearAccScale=<1 1 1>
    componentLinearVelScale=<1 1 1>
    gravityOverride {z=-980}
    physicsBodyDefinitions [{
      boundBone {boneName=dyn_hair_bk_02}
      boxExtents=<30 40 40>
      constraintSetup {angularLimitsMax=<0 -1 0>, angularLimitsMin=<0 -5 0>, angularTarget=<1 0 0>}
      localJointOffset=<4 0 0>
     }]
   }
   n25 = Anim Dynamics - Bone: dyn_hair_bk_r_02!(ComponentPose=@n24, Alpha=1, GravityOverride:Vector="0,0,-980", ExternalForce="0,0,0") -> (Pose) {
    angularDampingOverride=0.9999
    angularSpringConstant=5000
    bAngularSpring=true
    bOverrideAngularDamping=true
    bUseGravityOverride=true
    gravityScale=0.35
    simulationSpace=RootRelative
    boundBone {boneName=dyn_hair_bk_r_02}
    componentLinearAccScale=<1 1 1>
    componentLinearVelScale=<1 1 1>
    gravityOverride {z=-980}
    physicsBodyDefinitions [{
      boundBone {boneName=dyn_hair_bk_r_02}
      boxExtents=<30 40 40>
      constraintSetup {angularLimitsMax=<0 -1 0>, angularLimitsMin=<0 -5 0>, angularTarget=<1 0 0>}
      localJointOffset=<4 0 0>
     }]
   }
   n26 = Anim Dynamics - Bone: dyn_hair_02!(ComponentPose=@n25, Alpha=1, GravityOverride:Vector="0,0,-980", ExternalForce) -> (Pose) {
    angularDampingOverride=0.999
    angularSpringConstant=12000
    bAngularSpring=true
    bOverrideAngularDamping=true
    bUseGravityOverride=true
    gravityScale=0.35
    simulationSpace=RootRelative
    boundBone {boneName=dyn_hair_02}
    componentLinearAccScale=<1 1 1>
    componentLinearVelScale=<1 1 1>
    gravityOverride {z=-980}
    physicsBodyDefinitions [{
      boundBone {boneName=dyn_hair_02}
      boxExtents=<40 40 40>
      constraintSetup {angularLimitsMax=<0 5 2>, angularLimitsMin=<0 -10 12>, angularTarget=<1 0 0>}
      localJointOffset=<4 0 0>
     }]
   }
   n27 = Anim Dynamics - Bone: dyn_hair_03!(ComponentPose=@n26, Alpha=1, GravityOverride:Vector="0,0,-980", ExternalForce="0,0,0") -> (Pose) {
    angularDampingOverride=0.999
    angularSpringConstant=12000
    bAngularSpring=true
    bOverrideAngularDamping=true
    bUseGravityOverride=true
    gravityScale=0.35
    simulationSpace=RootRelative
    boundBone {boneName=dyn_hair_03}
    componentLinearAccScale=<1 1 1>
    componentLinearVelScale=<1 1 1>
    gravityOverride {z=-980}
    physicsBodyDefinitions [{
      boundBone {boneName=dyn_hair_03}
      boxExtents=<40 40 40>
      constraintSetup {angularLimitsMax=<0 5 2>, angularLimitsMin=<0 -10 12>, angularTarget=<1 0 0>}
      localJointOffset=<4 0 0>
     }]
   }
   n28 = Anim Dynamics - Bone: dyn_hair_04!(ComponentPose=@n27, Alpha=1, GravityOverride:Vector="0,0,-980", ExternalForce="0,0,0") -> (Pose) {
    angularDampingOverride=0.999
    angularSpringConstant=12000
    bAngularSpring=true
    bOverrideAngularDamping=true
    bUseGravityOverride=true
    gravityScale=0.35
    simulationSpace=RootRelative
    boundBone {boneName=dyn_hair_04}
    componentLinearAccScale=<1 1 1>
    componentLinearVelScale=<1 1 1>
    gravityOverride {z=-980}
    physicsBodyDefinitions [{
      boundBone {boneName=dyn_hair_04}
      boxExtents=<40 40 40>
      constraintSetup {angularLimitsMax=<0 5 2>, angularLimitsMin=<0 -10 12>, angularTarget=<1 0 0>}
      localJointOffset=<4 0 0>
     }]
   }
   n29 = Anim Dynamics - Bone: dyn_hair_05!(ComponentPose=@n28, Alpha=1, GravityOverride:Vector="0,0,-980", ExternalForce="0,0,0") -> (Pose) {
    angularDampingOverride=0.999
    angularSpringConstant=12000
    bAngularSpring=true
    bOverrideAngularDamping=true
    bUseGravityOverride=true
    gravityScale=0.35
    simulationSpace=RootRelative
    boundBone {boneName=dyn_hair_05}
    componentLinearAccScale=<1 1 1>
    componentLinearVelScale=<1 1 1>
    gravityOverride {z=-980}
    physicsBodyDefinitions [{
      boundBone {boneName=dyn_hair_05}
      boxExtents=<40 40 40>
      constraintSetup {angularLimitsMax=<0 5 2>, angularLimitsMin=<0 -10 12>, angularTarget=<1 0 0>}
      localJointOffset=<4 0 0>
     }]
   }
   n30 = Anim Dynamics - Bone: dyn_hair_07!(ComponentPose=@n29, Alpha=1, GravityOverride:Vector="0,0,-980", ExternalForce="0,0,0") -> (Pose) {
    angularDampingOverride=0.999
    angularSpringConstant=12000
    bAngularSpring=true
    bOverrideAngularDamping=true
    bUseGravityOverride=true
    gravityScale=0.35
    simulationSpace=RootRelative
    boundBone {boneName=dyn_hair_07}
    componentLinearAccScale=<1 1 1>
    componentLinearVelScale=<1 1 1>
    gravityOverride {z=-980}
    physicsBodyDefinitions [{
      boundBone {boneName=dyn_hair_07}
      boxExtents=<40 40 40>
      constraintSetup {angularLimitsMax=<0 5 2>, angularLimitsMin=<0 -10 12>, angularTarget=<1 0 0>}
      localJointOffset=<4 0 0>
     }]
   }
   n31 = Anim Dynamics - Bone: dyn_hair_06!(ComponentPose=@n30, Alpha=1, GravityOverride:Vector="0,0,-980", ExternalForce="0,0,0") -> (Pose) {
    angularDampingOverride=0.999
    angularSpringConstant=12000
    bAngularSpring=true
    bOverrideAngularDamping=true
    bUseGravityOverride=true
    gravityScale=0.35
    simulationSpace=RootRelative
    boundBone {boneName=dyn_hair_06}
    componentLinearAccScale=<1 1 1>
    componentLinearVelScale=<1 1 1>
    gravityOverride {z=-980}
    physicsBodyDefinitions [{
      boundBone {boneName=dyn_hair_06}
      boxExtents=<40 40 40>
      constraintSetup {angularLimitsMax=<0 5 2>, angularLimitsMin=<0 -10 12>, angularTarget=<1 0 0>}
      localJointOffset=<4 0 0>
     }]
   }
   n32 = Anim Dynamics - Bone: dyn_hair_08!(ComponentPose=@n31, Alpha=1, GravityOverride:Vector="0,0,-980", ExternalForce="0,0,0") -> (Pose) {
    angularDampingOverride=0.999
    angularSpringConstant=12000
    bAngularSpring=true
    bOverrideAngularDamping=true
    bUseGravityOverride=true
    gravityScale=0.35
    simulationSpace=RootRelative
    boundBone {boneName=dyn_hair_08}
    componentLinearAccScale=<1 1 1>
    componentLinearVelScale=<1 1 1>
    gravityOverride {z=-980}
    physicsBodyDefinitions [{
      boundBone {boneName=dyn_hair_08}
      boxExtents=<40 40 40>
      constraintSetup {angularLimitsMax=<0 5 2>, angularLimitsMin=<0 -10 12>, angularTarget=<1 0 0>}
      localJointOffset=<4 0 0>
     }]
   }
   n33 = Control Rig!(
    Source=Component To Local(ComponentPose=@n32)
    Alpha
    Head_L:double=Get HeadL() -> (HeadL:double)
    Head_R:double=Get HeadR() -> (HeadR:double)
    Head_Up:double=Get HeadUP() -> (HeadUP:double)
    HeadDown:double=Get HeadDown() -> (HeadDown:double)
    isAiming:double=Get TargetingFloat() -> (TargetingFloat:double)
    isTacticalJumping:double=Get TacticalJumping() -> (TacticalJumping:double)
    isJumping:double=Get IsJumpingFloat() -> (IsJumpingFloat:double)
    isCrouching:double=Get IsCrouchingFloat() -> (IsCrouchingFloat:double)
    isSkydiving:double=Get skydivingFloat() -> (skydivingFloat:double)
    isEmoting:double=Get isEmotingFloat() -> (isEmotingFloat:double)
    FrontEnd:bool=Get bIsInFrontEnd()
   ) -> (Pose) {controlRigAssetReference {blueprintRigClass=&/F_MED_QuarterClaspJoin_FaceAcc_Post_CtrlRig.F_MED_QuarterClaspJoin_FaceAcc_Post_CtrlRig_C}}
   Output Pose(Result=@n33) {layerGroup=None}
  ]}
 EventGraph {nodes [
   Event BlueprintUpdateAnimation()
   > n1 = Sequence!(execute) -> (then_0, then_1, then_2:Exec, then_3:Exec)
   n2 = GetOwnerSkeletalMeshComponent()
   n3 = BreakRotator(InRot=BreakTransform(InTransform=GetDeltaTransformFromRefPose(self=@n2, BoneName=head, BaseName=spine_05)).Rotation)
   Set HeadL(execute=@n1.then_1, HeadL:double=MapRangeClamped(Value=@n3, InRangeA=0, InRangeB=75)) -> (Output_Get:double)
   > Set HeadR(HeadR:double=MapRangeClamped(Value=@n3, InRangeA=0, InRangeB=-75)) -> (Output_Get:double)
   > Set HeadUP(HeadUP:double=MapRangeClamped(Value=@n3.Yaw, InRangeA=0, InRangeB=-75)) -> (Output_Get:double)
   > Set HeadDown(HeadDown:double=MapRangeClamped(Value=@n3.Yaw, InRangeA=0, InRangeB=30)) -> (Output_Get:double)
   n4 = "Gravity Override State Function (Parameter Asset) (Nativized)"(
    execute=@n1.then_0
    CharacterPartAnimInstance=Self-Reference()
    GameTimeInSeconds=GetGameTimeInSeconds()
    Parameters=&/F_MED_QuarterClaspJoin_FaceAcc_Parameters
   )
   > Set GravityOverride(GravityOverride=@n4)
   > Set RBAN Settings Out(RBAN Settings Out=@n4.OutRigidBodyAnimNodeParameters)
   Set TargetingFloat(execute=@n1.then_2, TargetingFloat:double="To Float (Boolean)"(InBool=Get bIsTargeting())) -> (Output_Get:double)
   n5 = AND Boolean(A=Get bIsInAir() -> (bIsInAir:bool), B=NOT Boolean(A=Get bIsSkyDiving()))
   n6 = AND Boolean(A=Get bIsTacticalSprinting(), B=@n5)
   > Set TacticalJumping(TacticalJumping:double="To Float (Boolean)"(InBool=@n6)) -> (Output_Get:double)
   n7 = NOT Boolean(A=@n6)
   > Set IsJumpingFloat(IsJumpingFloat:double="To Float (Boolean)"(InBool=AND Boolean(A=@n7, B=@n5))) -> (Output_Get:double)
   > Set IsCrouchingFloat(IsCrouchingFloat:double="To Float (Boolean)"(InBool=Get bIsCrouching())) -> (Output_Get:double)
   > Set skydivingFloat(skydivingFloat:double="To Float (Boolean)"(InBool=Get bIsSkyDiving())) -> (Output_Get:double)
   > Set isEmotingFloat(isEmotingFloat:double="To Float (Boolean)"(InBool=Get bIsPlayingEmote())) -> (Output_Get:double)
   n8 = Event OnCharacterCustomizationLoaded()
   > Set Body(Body:Skeletal Mesh Component Budgeted Object Reference=Get Array Item with EFort Custom Part Type(Array=@n8.CustomMeshParts)) -> (Output_Get:Skeletal Mesh Component Budgeted Object Reference)
   Event BlueprintInitializeAnimation()
   > Set Skeletal_Mesh(Skeletal_Mesh:Skeletal Mesh Component Object Reference=GetOwnerSkeletalMeshComponent()) -> (Output_Get:Skeletal Mesh Component Object Reference)
  ]}
]
variables [
 Body {type=SkeletalMeshComponentBudgeted}
 GravityOverride {type=Vector}
 HeadDown {type=number}
 HeadL {type=number}
 HeadR {type=number}
 HeadUP {type=number}
 IsCrouchingFloat {type=number}
 isEmotingFloat {type=number}
 IsJumpingFloat {type=number}
 RBAN Settings Out {type=FortRigidBodyAnimNodeParameters}
 Skeletal_Mesh {type=SkeletalMeshComponent}
 skydivingFloat {type=number}
 TacticalJumping {type=number}
 TargetingFloat {type=number}
]
