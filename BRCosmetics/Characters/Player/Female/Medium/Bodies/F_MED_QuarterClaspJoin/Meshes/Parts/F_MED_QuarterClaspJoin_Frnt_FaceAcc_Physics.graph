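# Physics asset for the front face-accessory part of the female medium QuarterClaspJoin cosmetic, defining capsule bodies for its dyn_hair bones.
# The chain runs from head through dyn_hair_parent to 03, branches to 04, 05 and 06, then from 05 to 07 and 08.
# Swing limits mostly widen down the chain but tighten on 06–08, with zero twist limits.
# Head, neck, upper-spine and left-only shoulder and arm capsules provide body collision volumes.
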PhysicsAsset /BRCosmetics/Characters/Player/Female/Medium/Bodies/F_MED_QuarterClaspJoin/Meshes/Parts/F_MED_QuarterClaspJoin_Frnt_FaceAcc_Physics
bodies {
 clavicle_l [{shapeType=Capsule, length=12, radius=5.5, shapeName=clavicle_l_capsule, center=<3.5 -1 -1>, rotation=<rot 80 -0.0003 -0.0003>}]
 dyn_hair_01 [{shapeType=Capsule, length=1.6001, radius=3.7736, center=<2.9625 -0.7268 -0.2359>, rotation=<rot 82.6643 0.0001 0>}]
 dyn_hair_02 [{shapeType=Capsule, length=3.4879, radius=3.3714, center=<3.5095 -0.2364 -0.1404>, rotation=<rot 74.018 -160.0004 178.588>}]
 dyn_hair_03 [{shapeType=Capsule, length=3.052, radius=4.1467, center=<3.1366 -0.3449 0>, rotation=<rot 90 0 0>}]
 dyn_hair_04 [{shapeType=Capsule, length=2.6191, radius=3.3967, center=<1.9598 1.0644 -1.1536>, rotation=<rot 80 0 -0.0003>}]
 dyn_hair_05 [{shapeType=Capsule, length=2.6951, radius=3.7882, center=<4.2705 -0.8596 -0.6887>, rotation=<rot 90 -18.527 0>}]
 dyn_hair_06 [{shapeType=Capsule, length=3, radius=1.3, center=<2.9058 0 0>, rotation=<rot 90 0 0>}]
 dyn_hair_07 [{shapeType=Capsule, length=2.3, radius=1.3, center=<2.5508 0 0>, rotation=<rot 90 0 0>}]
 dyn_hair_08 [{shapeType=Capsule, length=1.505, radius=1, center=<1.7662 0 0>, rotation=<rot 90 0 0>}]
 dyn_hair_parent [{shapeType=Capsule, length=3.0654, radius=2.866, center=<2.707 0 -0.298>, rotation=<rot 70 -0.0001 9.9997>}]
 head [{shapeType=Capsule, length=4.8682, radius=7.5954, shapeName=head_capsule, center=<3.5 3 0>, rotation=<rot 90 0 0>}]
 neck_01 [{
   shapeType=Capsule
   length=3.2426
   radius=5.1678
   shapeName=neck_01_capsule
   center=<3.2212 -1.1775 -0.0112>
   rotation=<rot -1.3479 159.9973 -179.5094>
  }]
 spine_04 [{shapeType=Capsule, length=11.375, radius=18.625, center=<-6.5908 -7.8763 -0.9615>, rotation=<rot 89.3496 179.9999 180>}]
 spine_05 [
  {shapeType=Capsule, length=10.3169, radius=9.25, shapeName=spine_05_capsule, center=<-2.5963 -0.392 -1.366>}
  {shapeType=Capsule, length=12, radius=7.75, shapeName=spine_05_capsule, center=<-6.4845 2.0672 0>}
  {shapeType=Capsule, length=23.4415, radius=6.5585, shapeName=spine_05_capsule, center=<12.1032 -4.1347 -3.3633>}
 ]
 upperarm_l [{
   shapeType=Capsule
   length=25.905
   radius=5.06
   shapeName=upperarm_l_capsule
   center=<17.7139 -0.2173 2.2995>
   rotation=<rot -0.0001 -89.9999 84.2668>
  }]
}
constraints [
 {bone1Name=dyn_hair_02, bone2Name=dyn_hair_01, swing1LimitDegrees=20, swing2LimitDegrees=10, twistLimitDegrees=0}
 {bone1Name=dyn_hair_03, bone2Name=dyn_hair_02, swing1LimitDegrees=25, swing2LimitDegrees=15, twistLimitDegrees=0}
 {bone1Name=dyn_hair_04, bone2Name=dyn_hair_03, swing1LimitDegrees=30, swing2LimitDegrees=20, twistLimitDegrees=0}
 {bone1Name=dyn_hair_05, bone2Name=dyn_hair_03, swing1LimitDegrees=35, swing2LimitDegrees=25, twistLimitDegrees=0}
 {bone1Name=dyn_hair_06, bone2Name=dyn_hair_03, swing1LimitDegrees=20, swing2LimitDegrees=5, twistLimitDegrees=0}
 {bone1Name=dyn_hair_07, bone2Name=dyn_hair_05, swing1LimitDegrees=20, swing2LimitDegrees=5, twistLimitDegrees=0}
 {bone1Name=dyn_hair_08, bone2Name=dyn_hair_05, swing1LimitDegrees=20, swing2LimitDegrees=5, twistLimitDegrees=0}
 {bone1Name=dyn_hair_01, bone2Name=dyn_hair_parent, swing1LimitDegrees=15, swing2LimitDegrees=5, twistLimitDegrees=0}
 {bone1Name=dyn_hair_parent, bone2Name=head, swing1LimitDegrees=10, swing2LimitDegrees=2, twistLimitDegrees=0}
 {
  bone1Name=upperarm_l
  bone2Name=clavicle_l
  swing1LimitDegrees=45
  swing1Motion=Free
  swing2LimitDegrees=45
  swing2Motion=Free
  twistLimitDegrees=45
  twistMotion=Free
 }
]
